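# Unlit, two-sided translucent particle material for the airglow on the knockback weapon's ring projectile.
# Particle color, scaled by time-of-day values, feeds emissive adjustments.
# Opacity multiplies a powered, polar-mapped mask by Fresnel, particle alpha, near-camera and depth fades, with dynamic parameters tuning power, Fresnel and fade distance.
# The arctangent takes G for both inputs, and FresnelBrightness also drives the exponent, leaving FresnelExponent unused.
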Material /Skyfire/Gameplay/KnockbackWeapon/FX/RingProj/M_Knockback_Proj_Airglow
graph {
 properties {
  bUsedWithMeshParticles=true
  bUsedWithNiagaraSprites=true
  bUsedWithParticleSprites=true
  blendMode=BLEND_Translucent
  pixelDepthOffsetMode=PDOM_Legacy
  shadingModel=MSM_Unlit
  twoSided=true
  usedShadingModels=Unlit
 }
 nodes [
  n1 = ParticleColor()
  n2 = ScalarParameter() {defaultValue=1, parameterName="Value_Morning (S)"}
  n3 = ScalarParameter() {defaultValue=1, parameterName="Value_Day (S)"}
  n4 = ScalarParameter() {defaultValue=1, parameterName="Value_Evening (S)"}
  n5 = ScalarParameter() {defaultValue=1, parameterName="Value_Night (S)"}
  MaterialFunctionCall(Emissive=Multiply(
    A=@n1.RGB
    B=MaterialFunctionCall(Value_Morning=@n2, Value_Day=@n3, Value_Evening=@n4, Value_Night=@n5) {materialFunction=&Scalar_Time_of_Day}.Result
   )) {materialFunction=&MF_FX_EmissiveAdjustments}
  n6 = Subtract(A=TextureCoordinate()) {constB=0.5}
  n7 = MaterialFunctionCall(Float2=@n6) {materialFunction=&BreakOutFloat2Components}
  n8 = Arctangent2Fast(Y=@n7.G, X=@n7.G)
  n9 = MaterialFunctionCall() {materialFunction=&Pi}
  n10 = Constant() {r=2}
  n11 = AppendVector(
   A=SquareRoot(None=MaterialFunctionCall(Vector 2=@n6) {materialFunction=/Engine/Functions/Engine_MaterialFunctions02/Utility/VectorLength}.V2 Length)
   B=Divide(
    A=Add(A=@n8, B=@n9.Result)
    B=MaterialFunctionCall(Multiplier=@n10) {materialFunction=&Pi}.Result
   )
  )
  n12 = MaterialFunctionCall(XY=@n11) {materialFunction=&Swizzle}
  n13 = ScalarParameter() {defaultValue=1, group="1.1) Radial Mask Rotation", parameterName=Radial Mask - Rotate UVs, sortPriority=3}
  n14 = Constant() {r=1}
  n15 = Constant() {r=-1}
  n16 = LinearInterpolate(A=@n14, B=@n15, Alpha=ScalarParameter() {group="1.1) Radial Mask Rotation", parameterName=Radial Mask - Flip UVs Horizontal, sortPriority=4})
  n17 = LinearInterpolate(A=@n14, B=@n15, Alpha=ScalarParameter() {group="1.1) Radial Mask Rotation", parameterName=Radial Mask - Flip UVs Vertically, sortPriority=5})
  n18 = TextureCoordinate()
  n19 = TextureSampleParameter2D!(
   UVs=Multiply(A=LinearInterpolate(A=@n11, B=@n12.YX, Alpha=@n13), B=AppendVector(A=@n16, B=@n17))
   "DDX(UVs)"=DDX(Value=@n18)
   "DDY(UVs)"=DDY(Value=@n18)
   Apply View MipBias
  ) {
   constMipValue=0
   group=Radial Mask Texture
   mipValueMode=TMVM_Derivative
   parameterName=Radial Mask
   sortPriority=70
   texture=/Game/Global/Textures/Default/Masks_and_Gradients/T_Water_Top_Ramp
  }
  n20 = DynamicParameter() {defaultValue=<rgba 3 1 1 1>, paramNames [Power, DepthFadeDistMult, FresnelBrightness, FresnelExponent]}
  n21 = Clamp(None=@n20.Power) {clampMode=CMODE_ClampMin, maxDefault=100, minDefault=1}
  n22 = MaterialFunctionCall(Distance=ScalarParameter() {defaultValue=100, parameterName=NearCameraFadeDist}) {materialFunction=&Near_Camera_Fade}
  n23 = Multiply(A=@n20.DepthFadeDistMult, B=ScalarParameter() {defaultValue=100, parameterName=DepthFadeDist})
  Multiply(
   A=Multiply(
    A=Power(Base=@n19.RGB, Exp=@n21) {constExponent=3}
    B=MaterialFunctionCall(Brightness=@n20.FresnelBrightness, Exponent=@n20.FresnelBrightness) {materialFunction=&MF_Fresnel_FX}.Fresnel
   )
   B=Multiply(A=@n1.A, B=DepthFade(Opacity=@n22.Result, FadeDistance=@n23))
  )
 ]
}
