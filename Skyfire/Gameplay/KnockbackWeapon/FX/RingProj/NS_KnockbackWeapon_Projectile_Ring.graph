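# Niagara system for the knockback weapon's ring-shaped projectile effect.
# A pulsing mesh ring sits with an owner-aligned airglow sprite, a curl-noise ribbon circle and a stretched trail mesh.
# RingInXRate streams smaller rings backward along X, rippling by sine and shrinking with distance while fading over life.
# User RingSize and colors drive the layers, ADS brightens the mesh rings; Sparks is disabled and ProjectileScale is unreferenced.
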NiagaraSystem /Skyfire/Gameplay/KnockbackWeapon/FX/RingProj/NS_KnockbackWeapon_Projectile_Ring
dependencies {
 usedDataInterfaces [&NiagaraDataInterfaceCurve, &NiagaraDataInterfaceVectorCurve]
 usedDynamicInputs []
 usedModules [&SystemState]
}
emitters {
 AirglowRing {
  emitterUpdateScript [
   &EmitterState()
   &SpawnBurst_Instantaneous(Spawn Count=1, Use Spawn Probability=false)
  ]
  particleSpawnScript [
   &InitializeParticle(
    Color=$User.Color_01
    Lifetime=3
    Mesh Scale Mode=::NewEnumerator3
    Position Offset=<0 0 1>
    Sprite Size Mode=::NewEnumerator3
    Uniform Sprite Size=&Multiply_Float(A=$User.RingSize, B=180)
    UsePositionOffset=true
    Write Material Random=true
   )
   /NiagaraFN/Sprite/FN_SpriteAlignmentAndFacing(SpriteFacing:Vector3f=&RotateVectorByQuaternion(Vector To Rotate=<1 0 0>, Quaternion=$Engine.Owner.Rotation))
   /SetVariables(Particles.Facing2:Vector3f=&RotateVectorByQuaternion(Vector To Rotate=<0 0 -1>, Quaternion=$Engine.Owner.Rotation)) {isSetParameters=true}
  ]
  particleUpdateScript [
   &ParticleState()
   &DynamicMaterialParameters(
    Index 0 Param 1=5
    Index 0 Param 2=1
    Index 0 Param 3=5
    Index 0 Param 4=3
    Module.Param0WriteEnabled=true
    Module.Param1WriteEnabled=true
    Module.Param2WriteEnabled=true
    Module.Param3WriteEnabled=true
   )
   &ScaleColor(Scale Alpha=0.2, ScaleA=true, ScaleRGB=true)
   &ScaleSpriteSize(Scale Factor=&Vector2DFromFloat(Value=&FloatFromCurve(FloatCurve=DataInterface() {propertyValues {Curve {keys [
          {arriveTangent=4.1785, interpMode=RCIM_Cubic, leaveTangent=4.1785, tangentMode=RCTM_User, time=0, value=0.25}
          {arriveTangent=0.9995, interpMode=RCIM_Cubic, leaveTangent=0.9995, tangentMode=RCTM_User, time=0.2223, value=0.9}
          {arriveTangent=-0.0405, interpMode=RCIM_Cubic, leaveTangent=-0.0405, tangentMode=RCTM_User, time=1, value=1}
         ]}}}))) {enabled=false}
   &ApplyOwnerScaleToAttributes(Scale Initial Sprite Size=true) {enabled=false}
   &ScaleSpriteSize(Scale Factor=&Vector2DFromFloat(Value=&MakeFloatFromVector(Vector=$Engine.Owner.Scale, Channel=::NewEnumerator0))) {module=ScaleSpriteSize001}
  ]
  rendererClasses [&NiagaraSpriteRendererProperties]
  renderers [
   {rendererClass=&NiagaraSpriteRendererProperties}
   {rendererClass=&NiagaraSpriteRendererProperties, rendererIndex=1}
  ]
 }
 Ribbon {
  emitterUpdateScript [
   &EmitterState()
   &SpawnBurst_Instantaneous(Spawn Count=25, Use Spawn Probability=false)
   &SpawnRate(SpawnRate=5, Use Spawn Probability=false) {enabled=false}
  ]
  particleSpawnScript [
   &InitializeParticle(
    Color=&LerpLinearColors(StartColor=$User.Color_01, EndColor=$User.Color_02, LerpFactor=&RandomRangeFloat(Minimum=0, Maximum=1))
    Lifetime=3
    Mass Max=2
    Mass Mode=::NewEnumerator3
    Sprite Size Max=<2 8>
    Sprite Size Min=<1.5 6>
    Sprite Size Mode=::NewEnumerator2
    UsePositionOffset=false
    Write Material Random=true
   ) {enabled=false}
   &InitializeRibbon(
    Color=&LerpLinearColors(StartColor=$User.Color_01, EndColor=$User.Color_02, LerpFactor=&RandomRangeFloat(Minimum=0, Maximum=1))
    Lifetime=3
    Module.Write Color=true
    Module.Write Lifetime=true
    Module.Write Mass=true
    Module.Write Material Random=true
    Module.Write Position=false
    Module.Write RibbonTwist=false
    Module.Write RibbonWidth=true
    Ribbon Width=&RandomRangeFloat(Minimum=10, Maximum=12)
   )
   &AddVelocityInCone(
    Cone Angle=48
    Cone Axis=<0.9 0 0.4>
    Module.Use Velocity Falloff On Cone Axis=true
    Velocity Distribution Along Cone Axis=0.75
    Velocity Falloff Away From Cone Axis=0.333
    Velocity Strength=&RandomRangeFloat(Minimum=350, Maximum=750)
   ) {enabled=false}
   &CylinderLocation(
    Cylinder Height=0
    Cylinder Radius=&Multiply_Float(A=$User.RingSize, B=50)
    MaskSpawnGroup=false
    Module.Enable Lathe Profile=false
    Module.Override Randomness=false
    Module.Surface Only=true
    Offset=<1 0 0>
    Orientation Axis:ENiagaraOrientationAxis=::XAxis
    Surface Only Band Thickness=0
    Use Endcaps In Surface Only Mode=false
   )
   &InheritVelocity(Module.Limit Speed=true, Velocity Limit=250, Velocity Scale=<-1 -1 -1>) {enabled=false}
   &SphereLocation(MaskSpawnGroup=false, Override Seed=false, Surface Only=false, UseNonUniformScale=false, UseOffset=false, UseSphereOrientation=false) {enabled=false}
   &FN_SetSpawnGroup(Number of Spawn Groups=1, Set Ribbon ID and Ribbon Link Order=true)
   &AddVelocityFromPoint(Module.Use Falloff Distance=false, Velocity Strength=475) {enabled=false}
  ]
  particleUpdateScript [
   &ParticleState()
   &GravityForce(Gravity=<0 0 -200>) {enabled=false}
   &Drag(
    Drag=&FloatFromCurve(
     FloatCurve=DataInterface() {propertyValues {Curve {keys [{interpMode=RCIM_Cubic, tangentMode=RCTM_User, time=0, value=0}, {interpMode=RCIM_Cubic, tangentMode=RCTM_User, time=0.5, value=2}]}}}
     Scale Curve=50
    )
    Use Linear Drag=true
    Use Rotational Drag=false
   )
   &CurlNoiseForce(
    Noise Frequency=150
    Noise Strength=400
    Pan Noise=true
    Pan Noise Field=&MakeVector(
     X=0
     Y=0
     Z=&FloatFromCurve(
      FloatCurve=DataInterface() {propertyValues {Curve {keys [{interpMode=RCIM_Cubic, tangentMode=RCTM_User, time=0, value=2}, {interpMode=RCIM_Cubic, tangentMode=RCTM_User, time=1, value=1}]}}}
      Scale Curve=50000
     )
    )
   )
   &SolveForcesAndVelocity(Clamp Velocity=false, Limit Acceleration=false)
   &ScaleSpriteSizeBySpeed(Max Scale Factor=<1.2 5>, Velocity Threshold=1500) {enabled=false}
   &ScaleSpriteSize(Scale Factor=&Vector2DFromFloat(Value=&FloatFromCurve(FloatCurve=DataInterface() {propertyValues {Curve {keys [
          {arriveTangent=9.0725, interpMode=RCIM_Cubic, leaveTangent=9.0725, tangentMode=RCTM_User, time=0, value=0}
          {arriveTangent=-2.5249, interpMode=RCIM_Cubic, leaveTangent=-2.5249, tangentMode=RCTM_User, time=0.1408, value=0.9847}
          {arriveTangent=-0.0405, interpMode=RCIM_Cubic, leaveTangent=-0.0405, tangentMode=RCTM_User, time=1, value=0}
         ]}}}))) {enabled=false}
   &ScaleColor(Scale Alpha=&RandomRangeFloat(Minimum=1, Maximum=2), ScaleA=true, ScaleRGB=true)
   &ScaleRibbonWidth(Ribbon Width Scale=&FloatFromCurve(
     FloatCurve=DataInterface() {propertyValues {Curve {keys [{tangentMode=RCTM_User, time=0.9, value=0}, {tangentMode=RCTM_User, time=1, value=3}]}}}
     CurveIndex=&NormalizeDistanceRange(
      Start Position:NiagaraPosition=$Particles.Position
      End Position:NiagaraPosition=$Engine.Owner.Position
      Distance:NiagaraFloat=1
      Invert Normalized Range:NiagaraBool=false
     )
    )) {enabled=false}
   &ScaleColor(Color Value To Scale=$Particles.Color, Scale Alpha=220, ScaleA=true, ScaleRGB=true) {enabled=false, module=ScaleColor001}
  ]
  rendererClasses [&NiagaraSpriteRendererProperties, &NiagaraRibbonRendererProperties]
  renderers [
   {rendererClass=&NiagaraSpriteRendererProperties}
   {rendererClass=&NiagaraRibbonRendererProperties, rendererIndex=1}
  ]
 }
 RingInX {
  emitterUpdateScript [
   &EmitterState()
   &SpawnBurst_Instantaneous(Spawn Count=1, Use Spawn Probability=false)
  ]
  particleSpawnScript [
   &InitializeParticle(
    Color=$User.Color_01
    Lifetime=3
    Mesh Scale Mode=::NewEnumerator3
    Mesh Uniform Scale=&Multiply_Float(A=$User.RingSize, B=0.5)
    UsePositionOffset=false
    Write Material Random=true
   )
   &InitialMeshOrientation(
    Orientation Vector=<0 0 1>
    Rotation=&RandomRangeVector(Minimum=<0 0 0>, Maximum=<0 0 1>)
    Use Orientation Vector=true
    Use Rotation Vector=true
   )
   &AddVelocity(Velocity=&Multiply_VectorByFloat(Vector=<0 0 0>, Float=$User.RingSize)) {enabled=false}
  ]
  particleUpdateScript [
   &ParticleState()
   &SolveRotationalForcesAndVelocity() {enabled=false}
   /SetVariables(Particles.SizeWave:NiagaraFloat=&Waveform(
     Global Phase=$Particles.NormalizedAge
     "Waveform [1]"=::NewEnumerator0
     "[1] Phase Scale"=1.75
     "[1] Phase"=$Particles.NormalizedAge
     "[1] Frequency"=&Lerp_Float(A=0.4, B=1.1, Alpha=$User.ADS)
     "[1] Amplitude Min/Max"=<0 1>
     "[1] Scale"=0.5
     "[1] Bias"=0.5
     "[2] Phase"=$Emitter.Age
     "[3] Phase"=$Emitter.Age
    )) {enabled=false, isSetParameters=true, module=SetVariables_#1}
   &SolveForcesAndVelocity(Clamp Velocity=false, Limit Acceleration=false)
   &ScaleColor(
    Scale Alpha=2
    Scale RGB=&Multiply_VectorByFloat(Vector=<1 1 1>, Float=&Lerp_Float(A=2, B=3, Alpha=$User.ADS))
    ScaleA=true
    ScaleRGB=true
   )
   &DynamicMaterialParameters(
    Index 0 Param 1=&FloatFromCurve(
     FloatCurve=DataInterface() {propertyValues {Curve {keys [<at 0 1>,<at 1 0.1>]}}}
     CurveIndex=&FN_LOD_Distance_Remap(Start Distance=0, End Distance=1000, Minimum Output=0, Maximum Output=1)
     Scale Curve=10
    )
    Module.Param0WriteEnabled=true
    Module.Param1WriteEnabled=true
    Module.Param2WriteEnabled=true
    Module.Param3WriteEnabled=false
   )
   &ScaleMeshSize(Scale Factor=&Multiply_VectorByFloat(
     Vector=&Multiply_VectorByFloat(
      Vector=&MakeVector(X=$Particles.SizeWave, Y=$Particles.SizeWave, Z=1)
      Float=&FloatFromCurve(FloatCurve=DataInterface() {propertyValues {Curve {keys [
           {interpMode=RCIM_Cubic, tangentMode=RCTM_User, time=0, value=1}
           {arriveTangent=-0.5603, interpMode=RCIM_Cubic, tangentMode=RCTM_Break, time=1, value=0.8}
          ]}}})
     )
     Float=$User.RingSize
    )) {enabled=false}
   &ScaleMeshSize(Scale Factor=&VectorFromFloat(Value=&FloatFromCurve(FloatCurve=DataInterface() {propertyValues {Curve {keys [
          {arriveTangent=4.1785, interpMode=RCIM_Cubic, leaveTangent=4.1785, tangentMode=RCTM_User, time=0, value=0.25}
          {arriveTangent=0.4581, interpMode=RCIM_Cubic, leaveTangent=0.4581, tangentMode=RCTM_User, time=0.2223, value=0.9}
          {arriveTangent=-0.0405, interpMode=RCIM_Cubic, leaveTangent=-0.0405, tangentMode=RCTM_User, time=1, value=1}
         ]}}}))) {enabled=false, module=ScaleMeshSize001}
   &ScaleMeshSize(Scale Factor=&VectorFromFloat(Value=&FloatFromCurve(FloatCurve=DataInterface() {propertyValues {Curve {keys [
          <at 0 1 RCIM_Cubic>
          {arriveTangent=0.25, interpMode=RCIM_Cubic, leaveTangent=0.25, time=0.1, value=0.95}
          <at 0.2 1.05 RCIM_Cubic>
          {arriveTangent=-0.25, interpMode=RCIM_Cubic, leaveTangent=-0.25, time=0.3, value=0.95}
          {arriveTangent=0.1167, interpMode=RCIM_Cubic, leaveTangent=0.1167, time=0.4, value=1}
          {arriveTangent=0.125, interpMode=RCIM_Cubic, leaveTangent=0.125, time=0.6, value=0.985}
          {arriveTangent=0.0375, interpMode=RCIM_Cubic, leaveTangent=0.0375, time=0.8, value=1.05}
          <at 1 1 RCIM_Cubic>
         ]}}}))) {module=ScaleMeshSize002}
   &ScaleMeshSize(Scale Factor=&VectorFromFloat(Value=&RandomRangeFloat(Minimum=0.99, Maximum=1.01, Evaluation Type=::NewEnumerator1))) {module=ScaleMeshSize003}
  ]
  rendererClasses [&NiagaraMeshRendererProperties]
  renderers [
   {rendererClass=&NiagaraMeshRendererProperties, rendererIndex=0}
   {rendererClass=&NiagaraMeshRendererProperties, rendererIndex=1}
  ]
 }
 RingInXRate {
  emitterUpdateScript [
   &EmitterState()
   &SpawnBurst_Instantaneous(Spawn Count=10, Use Spawn Probability=false) {enabled=false}
   &SpawnRate(SpawnRate=80, Use Spawn Probability=false)
  ]
  particleSpawnScript [
   &InitializeParticle(
    Color=&RandomRangeLinearColor(Minimum=$User.Color_01, Maximum=$User.Color_02)
    Lifetime=3
    Mesh Scale Mode=::NewEnumerator3
    Mesh Uniform Scale=&Multiply_Float(A=$User.RingSize, B=&RandomRangeFloat(Minimum=0.3, Maximum=0.4))
    UsePositionOffset=false
    Write Material Random=true
   )
   &InitialMeshOrientation(
    Orientation Vector=<0 0 1>
    Rotation=&RandomRangeVector(Minimum=<0 0 0>, Maximum=<0 0 1>)
    Use Orientation Vector=true
    Use Rotation Vector=true
   )
   &AddVelocity(Velocity=&RotateVectorByQuaternion(Vector To Rotate=<-5000 0 0>, Quaternion=<0 0 0 1>))
   &CylinderLocation(
    Cylinder Height=200
    Cylinder Radius=0
    MaskSpawnGroup=false
    Module.Enable Lathe Profile=false
    Module.Override Randomness=false
    Module.Surface Only=false
    Offset=<-200 0 0>
    Orientation Axis:ENiagaraOrientationAxis=::XAxis
   )
  ]
  particleUpdateScript [
   &ParticleState()
   &SolveRotationalForcesAndVelocity() {enabled=false}
   /SetVariables(Particles.SizeWave:NiagaraFloat=&Waveform(
     Global Phase=$Particles.NormalizedAge
     "Waveform [1]"=::NewEnumerator0
     "[1] Phase Scale"=1.75
     "[1] Phase"=$Particles.NormalizedAge
     "[1] Frequency"=&Lerp_Float(A=0.4, B=1.1, Alpha=$User.ADS)
     "[1] Amplitude Min/Max"=<0 1>
     "[1] Scale"=0.5
     "[1] Bias"=0.5
     "[2] Phase"=$Emitter.Age
     "[3] Phase"=$Emitter.Age
    )) {enabled=false, isSetParameters=true, module=SetVariables_#1}
   &SolveForcesAndVelocity(Clamp Velocity=false, Limit Acceleration=false)
   &ScaleColor(
    Scale Alpha=2
    Scale RGB=&Multiply_VectorByFloat(Vector=<1 1 1>, Float=&Lerp_Float(A=2, B=3, Alpha=$User.ADS))
    ScaleA=true
    ScaleRGB=true
   )
   &DynamicMaterialParameters(
    Index 0 Param 1=&FloatFromCurve(
     FloatCurve=DataInterface() {propertyValues {Curve {keys [<at 0 1>,<at 1 0.1>]}}}
     CurveIndex=&FN_LOD_Distance_Remap(Start Distance=0, End Distance=10000, Minimum Output=0, Maximum Output=1)
     Scale Curve=10
    )
    Module.Param0WriteEnabled=true
    Module.Param1WriteEnabled=true
    Module.Param2WriteEnabled=true
    Module.Param3WriteEnabled=false
   )
   &ScaleMeshSize(Scale Factor=&Multiply_VectorByFloat(
     Vector=&Multiply_VectorByFloat(
      Vector=&MakeVector(X=$Particles.SizeWave, Y=$Particles.SizeWave, Z=1)
      Float=&FloatFromCurve(FloatCurve=DataInterface() {propertyValues {Curve {keys [
           {interpMode=RCIM_Cubic, tangentMode=RCTM_User, time=0, value=1}
           {arriveTangent=-0.5603, interpMode=RCIM_Cubic, tangentMode=RCTM_Break, time=1, value=0.8}
          ]}}})
     )
     Float=$User.RingSize
    )) {enabled=false}
   &ScaleMeshSize(Scale Factor=&VectorFromFloat(Value=&FloatFromCurve(FloatCurve=DataInterface() {propertyValues {Curve {keys [
          {arriveTangent=4.1785, interpMode=RCIM_Cubic, leaveTangent=4.1785, tangentMode=RCTM_User, time=0, value=0.25}
          {arriveTangent=0.4581, interpMode=RCIM_Cubic, leaveTangent=0.4581, tangentMode=RCTM_User, time=0.2223, value=0.9}
          {arriveTangent=-0.0405, interpMode=RCIM_Cubic, leaveTangent=-0.0405, tangentMode=RCTM_User, time=1, value=1}
         ]}}}))) {enabled=false, module=ScaleMeshSize001}
   &ScaleMeshSize(Scale Factor=&VectorFromFloat(Value=&FloatFromCurve(
      FloatCurve=DataInterface() {propertyValues {Curve {keys [
          <at -1 1 RCIM_Cubic>
          {arriveTangent=0.0625, interpMode=RCIM_Cubic, leaveTangent=0.0625, time=-0.5, value=0.9625}
          <at -0.2 1.05 RCIM_Cubic>
          {arriveTangent=-0.0833, interpMode=RCIM_Cubic, leaveTangent=-0.0833, time=0.1, value=0.9625}
          {arriveTangent=0.045, interpMode=RCIM_Cubic, leaveTangent=0.045, time=0.4, value=1}
          {arriveTangent=0.125, interpMode=RCIM_Cubic, leaveTangent=0.125, time=0.6, value=0.985}
          {arriveTangent=0.0375, interpMode=RCIM_Cubic, leaveTangent=0.0375, time=0.8, value=1.05}
          <at 1 1 RCIM_Cubic>
         ]}}}
      CurveIndex=&Sine(
       Normalized Angle=&Subtract_Float(
        A=&RemapRange(
         Input Value To Remap=&VectorLength(Vector=&Subtract_Vector(A=/Niagara/DynamicInputs/Helpers/SimulationPosition(), B=Linked() {linkedVariable {name=Particles.Position, type=NiagaraPosition}}))
         Input Min=0
         Input Max=400
         Output Min=0
         Output Max=1
        )
        B=$Engine.Time
       )
       Period=5
       Scale=1
       Bias=0.5
      )
      Scale Curve=&RandomRangeFloat(Minimum=0.8, Maximum=1.2)
     ))) {module=ScaleMeshSize002}
   &ScaleMeshSize(Scale Factor=&VectorFromFloat(Value=&RandomRangeFloat(Minimum=0.99, Maximum=1.01, Evaluation Type=::NewEnumerator1))) {module=ScaleMeshSize003}
   &ScaleMeshSize(Scale Factor=&VectorFromFloat(Value=&FloatFromCurve(
      FloatCurve=DataInterface() {propertyValues {Curve {keys [{interpMode=RCIM_Cubic, tangentMode=RCTM_User, time=0, value=1}, {interpMode=RCIM_Cubic, tangentMode=RCTM_User, time=1, value=0.2}]}}}
      CurveIndex=&RemapRange(
       Input Value To Remap=&VectorLength(Vector=&Subtract_Vector(A=Linked() {linkedVariable {name=Particles.Position, type=NiagaraPosition}}, B=&SimulationPosition()))
       Input Min=0
       Input Max=2000
       Output Min=0
       Output Max=1
      )
     ))) {module=ScaleMeshSize004}
   &ScaleColor(
    Scale Alpha=&FloatFromCurve(FloatCurve=DataInterface() {propertyValues {Curve {keys [{interpMode=RCIM_Cubic, tangentMode=RCTM_User, time=0, value=1}, {interpMode=RCIM_Cubic, tangentMode=RCTM_User, time=1, value=0}]}}})
    ScaleA=true
    ScaleRGB=false
   ) {module=ScaleColor001}
  ]
  rendererClasses [&NiagaraMeshRendererProperties]
  renderers [
   {rendererClass=&NiagaraMeshRendererProperties, rendererIndex=0}
   {rendererClass=&NiagaraMeshRendererProperties, rendererIndex=1}
  ]
 }
 RingTrail {
  emitterUpdateScript [
   &EmitterState()
   &SpawnBurst_Instantaneous(Spawn Count=1, Use Spawn Probability=false)
  ]
  particleSpawnScript [
   &InitializeParticle(
    Color=$User.Color_01
    Lifetime=3
    Mesh Scale=&Multiply_VectorByFloat(Vector=<50 1.1 1.1>, Float=$User.RingSize)
    Mesh Scale Mode=::NewEnumerator4
    Position Offset=<10 0 0>
    UsePositionOffset=true
    Write Material Random=true
   ) {module=InitializeParticle001}
   &DynamicMaterialParameters(
    Index 0 Param 4=&RandomRangeFloat(Minimum=0, Maximum=1, Evaluation Type=::NewEnumerator1)
    Module.Param0WriteEnabled=false
    Module.Param1WriteEnabled=false
    Module.Param2WriteEnabled=false
    Module.Param3WriteEnabled=true
   ) {module=DynamicMaterialParameters001}
   &InitialMeshOrientation(Orientation Axis=<0 0 1>, Orientation Vector=<0 0 1>, Use Orientation Vector=true, Use Rotation Vector=true)
  ]
  particleUpdateScript [
   &ParticleState()
   &DynamicMaterialParameters(
    Index 0 Param 1=&FloatFromCurve(FloatCurve=DataInterface() {propertyValues {Curve {keys [<at 0 0>,<at 1 0>]}}})
    Index 0 Param 2=&FloatFromCurve(FloatCurve=DataInterface() {propertyValues {Curve {keys [<at 0 1 RCIM_Cubic>, {arriveTangent=-4.4839, leaveTangent=-4.4839, time=1, value=1}]}}})
    Index 0 Param 3=1
    Module.Param0WriteEnabled=true
    Module.Param1WriteEnabled=true
    Module.Param2WriteEnabled=true
    Module.Param3WriteEnabled=false
   )
   &Color(Color=&LerpLinearColors(
     StartColor=$User.Color_01
     EndColor=$User.Color_02
     LerpFactor=&FloatFromCurve(FloatCurve=DataInterface() {propertyValues {Curve {keys [
          {interpMode=RCIM_Cubic, tangentMode=RCTM_User, time=0, value=0.5}
          {interpMode=RCIM_Cubic, tangentMode=RCTM_User, time=0.5, value=0}
          {tangentMode=RCTM_User, time=1, value=0.5}
         ]}}})
    ))
   &ScaleColor(
    Color Value To Scale=$Particles.Color
    Scale Alpha=0.35
    Scale RGB=&VectorFromFloat(Value=1.2)
    ScaleA=true
    ScaleRGB=true
   ) {module=ScaleColor001}
   &ScaleColor(
    Color Value To Scale=$Particles.Color
    Scale Alpha=&FloatFromCurve(
     FloatCurve=DataInterface() {propertyValues {Curve {keys [
         {interpMode=RCIM_Cubic, tangentMode=RCTM_User, time=0, value=0}
         {interpMode=RCIM_Cubic, tangentMode=RCTM_User, time=0.5, value=1}
         {tangentMode=RCTM_User, time=1, value=0}
        ]}}}
     Scale Curve=0
    )
    Scale RGB=&VectorFromCurve(VectorCurve=DataInterface() {propertyValues {XCurve {keys [<at 0 1>,<at 1 1>]}, YCurve {keys [<at 0 1>,<at 1 1>]}, ZCurve {keys [<at 0 1>,<at 1 1>]}}})
    ScaleA=true
    ScaleRGB=false
   ) {enabled=false}
   &ScaleMeshSize(Scale Factor=&VectorFromFloat(Value=&FloatFromCurve(FloatCurve=DataInterface() {propertyValues {Curve {keys [
          {arriveTangent=4.1785, interpMode=RCIM_Cubic, leaveTangent=4.1785, tangentMode=RCTM_User, time=0, value=0.25}
          {arriveTangent=0.4581, interpMode=RCIM_Cubic, leaveTangent=0.4581, tangentMode=RCTM_User, time=0.2223, value=0.9}
          {arriveTangent=-0.0405, interpMode=RCIM_Cubic, leaveTangent=-0.0405, tangentMode=RCTM_User, time=1, value=1}
         ]}}}))) {enabled=false, module=ScaleMeshSize001}
   &ScaleMeshSize(Scale Factor=&VectorFromFloat(Value=&FloatFromCurve(FloatCurve=DataInterface() {propertyValues {Curve {keys [
          <at 0 1 RCIM_Cubic>
          {arriveTangent=0.25, interpMode=RCIM_Cubic, leaveTangent=0.25, time=0.1, value=0.95}
          <at 0.2 1.05 RCIM_Cubic>
          {arriveTangent=-0.25, interpMode=RCIM_Cubic, leaveTangent=-0.25, time=0.3, value=0.95}
          {arriveTangent=0.1167, interpMode=RCIM_Cubic, leaveTangent=0.1167, time=0.4, value=1}
          {arriveTangent=0.125, interpMode=RCIM_Cubic, leaveTangent=0.125, time=0.6, value=0.985}
          {arriveTangent=0.0375, interpMode=RCIM_Cubic, leaveTangent=0.0375, time=0.8, value=1.05}
          <at 1 1 RCIM_Cubic>
         ]}}}))) {module=ScaleMeshSize002}
   &ScaleMeshSize(Scale Factor=&VectorFromFloat(Value=&RandomRangeFloat(Minimum=0.99, Maximum=1.01, Evaluation Type=::NewEnumerator1))) {module=ScaleMeshSize003}
   &ScaleMeshSize(Scale Factor=&MakeVector(
     X=&FloatFromCurve(FloatCurve=DataInterface() {propertyValues {Curve {keys [<at 0 0>,<at 0.001 6>]}}}, CurveIndex=$Emitter.Age)
     Y=1
     Z=1
    )) {enabled=false}
   /SetVariables(
    Particles.Color_02:LinearColor=&MultiplyLinearColor(Linear Color=$Particles.Color, Scale Factor=<0.6 0.6 0.6 1>)
    Particles.MeshScale_02:Vector3f=&Multiply_Vector(A=$Particles.Scale, B=<1.25 0.95 0.95>)
   ) {isSetParameters=true}
  ]
  rendererClasses [&NiagaraMeshRendererProperties]
  renderers [
   {rendererClass=&NiagaraMeshRendererProperties, rendererIndex=0}
   {rendererClass=&NiagaraMeshRendererProperties, rendererIndex=1}
  ]
 }
 Sparks {
  bEnabled=false
  emitterUpdateScript [
   &EmitterState()
   &SpawnBurst_Instantaneous(Spawn Count=75, Use Spawn Probability=false) {enabled=false}
   &SpawnRate(SpawnRate=180, Use Spawn Probability=false)
  ]
  particleSpawnScript [
   &InitializeParticle(
    Color=&LerpLinearColors(StartColor=$User.Color_02, EndColor=$User.Color_03, LerpFactor=&RandomRangeFloat(Minimum=0, Maximum=1))
    Lifetime Max=1
    Lifetime Min=0.8
    Lifetime Mode=::NewEnumerator1
    Mass Max=2
    Mass Mode=::NewEnumerator3
    Sprite Size Max=<4 10>
    Sprite Size Min=<3 8>
    Sprite Size Mode=::NewEnumerator2
    UsePositionOffset=false
    Write Material Random=true
   )
   &AddVelocityInCone(
    Cone Angle=48
    Cone Axis=<0.9 0 0.4>
    Module.Use Velocity Falloff On Cone Axis=true
    Velocity Distribution Along Cone Axis=0.75
    Velocity Falloff Away From Cone Axis=0.333
    Velocity Strength=&RandomRangeFloat(Minimum=350, Maximum=750)
   ) {enabled=false}
   &CylinderLocation(
    Cylinder Height=0
    Cylinder Radius=&Multiply_Float(A=$User.RingSize, B=50)
    Lathe Profile=DataInterface() {propertyValues {Curve {keys [
        {interpMode=RCIM_Cubic, tangentMode=RCTM_User, time=0, value=0}
        {interpMode=RCIM_Cubic, tangentMode=RCTM_User, time=0.5, value=0}
        {tangentMode=RCTM_User, time=1, value=0}
       ]}}}
    MaskSpawnGroup=false
    Module.Enable Lathe Profile=false
    Module.Override Randomness=false
    Module.Surface Only=true
    Orientation Axis:ENiagaraOrientationAxis=::XAxis
    Surface Only Band Thickness=2
    Use Endcaps In Surface Only Mode=false
   )
   &InheritVelocity(Module.Limit Speed=true, Velocity Limit=250, Velocity Scale=<-1 -1 -1>) {enabled=false}
   &SphereLocation(MaskSpawnGroup=false, Override Seed=false, Surface Only=false, UseNonUniformScale=false, UseOffset=false, UseSphereOrientation=false) {enabled=false}
   &CurlNoiseForce(Noise Strength=1200, Pan Noise=false) {module=CurlNoiseForce001}
   &ApplyInitialForces(
    Apply Force to Position=true
    Apply Force to Velocity=false
    Apply Rotational Force to Rotational Velocity=false
    Positional Force Warmup Time=0.01
   )
  ]
  particleUpdateScript [
   &ParticleState()
   &GravityForce(Gravity=<0 0 -200>) {enabled=false}
   &Drag(Drag=&RandomRangeFloat(Minimum=0.8, Maximum=1.2), Use Linear Drag=true, Use Rotational Drag=false)
   &CurlNoiseForce(Noise Strength=1200, Pan Noise=false)
   &ScaleColor(
    Scale Alpha=&FloatFromCurve(FloatCurve=DataInterface() {propertyValues {Curve {keys [<at -0.0046 1.0113 RCIM_Cubic>,<at 1.0011 0.002 RCIM_Cubic>]}}}, Scale Curve=2)
    Scale RGB=&VectorFromFloat(Value=&FloatFromCurve(FloatCurve=DataInterface() {propertyValues {Curve {keys [
          {arriveTangent=9.0725, interpMode=RCIM_Cubic, leaveTangent=9.0725, tangentMode=RCTM_User, time=0, value=1}
          {arriveTangent=-2.5249, interpMode=RCIM_Cubic, leaveTangent=-2.5249, tangentMode=RCTM_User, time=0.1408, value=5}
          {arriveTangent=-0.0405, interpMode=RCIM_Cubic, leaveTangent=-0.0405, tangentMode=RCTM_User, time=1, value=1}
         ]}}}))
    ScaleA=true
    ScaleRGB=true
   )
   &SolveForcesAndVelocity(Clamp Velocity=false, Limit Acceleration=false)
   &ScaleSpriteSizeBySpeed(Max Scale Factor=<1.4 5>, Min Scale Factor=<1.2 1>, Velocity Threshold=1500)
   &ScaleSpriteSize(Scale Factor=&Vector2DFromFloat(Value=&FloatFromCurve(FloatCurve=DataInterface() {propertyValues {Curve {keys [{tangentMode=RCTM_User, time=0, value=1}, {tangentMode=RCTM_User, time=1, value=0}]}}})))
  ]
  rendererClasses [&NiagaraSpriteRendererProperties]
  renderers [{rendererClass=&NiagaraSpriteRendererProperties}]
 }
}
summary {
 systemName=NS_KnockbackWeapon_Projectile_Ring
 emitters [
  {emitterName=RingInX, rendererClasses [&NiagaraMeshRendererProperties]}
  {bEnabled=false, emitterName=Sparks, rendererClasses [&NiagaraSpriteRendererProperties]}
  {emitterName=AirglowRing, rendererClasses [&NiagaraSpriteRendererProperties]}
  {emitterName=Ribbon, rendererClasses [&NiagaraSpriteRendererProperties, &NiagaraRibbonRendererProperties]}
  {emitterName=RingTrail, rendererClasses [&NiagaraMeshRendererProperties]}
  {emitterName=RingInXRate, rendererClasses [&NiagaraMeshRendererProperties]}
 ]
 userVariables [
  User.ADS {type=NiagaraFloat, defaultValue=0, description=""}
  User.Color_01 {type=LinearColor, description="", defaultValue {struct=LinearColor, value=<rgba 0 1 0.8115 1>}}
  User.Color_02 {type=LinearColor, description="", defaultValue {struct=LinearColor, value=<rgba 1 0.0396 0.894 1>}}
  User.Color_03 {type=LinearColor, description="", defaultValue {struct=LinearColor, value=<rgba 0.0664 0.2855 0.625 1>}}
  User.ProjectileScale {type=NiagaraFloat, defaultValue=0, description=""}
  User.RingSize {type=NiagaraFloat, defaultValue=1, description=""}
 ]
}
