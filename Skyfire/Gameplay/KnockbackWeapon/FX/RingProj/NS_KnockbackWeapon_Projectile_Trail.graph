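# Niagara trail effect for the knockback weapon's ring projectile, drawn by a single ribbon emitter.
# Short-lived ribbon particles spawn at a steady rate at the simulation position, taking color from User.Color_01 and width from User.RingSize.
# They fade, narrow and receive a negatively scaled force from inherited source movement, alongside dynamic material parameters.
# Spawn-per-unit, velocity and curl-noise modules are disabled, and the other user parameters are unreferenced here.
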NiagaraSystem /Skyfire/Gameplay/KnockbackWeapon/FX/RingProj/NS_KnockbackWeapon_Projectile_Trail
dependencies {usedDataInterfaces [&NiagaraDataInterfaceCurve], usedDynamicInputs [], usedModules [&SystemState]}
emitters {Trail {
  emitterUpdateScript [
   &EmitterState()
   &SpawnRate(SpawnRate=20, Use Spawn Probability=false)
   &SpawnPerUnit(
    Max Movement Threshold=1
    Module.Use Max Movement Threshold=true
    Module.Use Movement Tolerance=false
    Spawn Spacing=2000
    Use Spawn Probability=false
   ) {enabled=false}
  ]
  particleSpawnScript [
   &InitializeRibbon(
    Color=$User.Color_01
    Lifetime=0.35
    Module.Write Color=true
    Module.Write Lifetime=true
    Module.Write Mass=true
    Module.Write Material Random=true
    Module.Write Position=true
    Module.Write RibbonTwist=false
    Module.Write RibbonWidth=true
    Position:Vector3f=&Add_Vector(
     A=/Niagara/DynamicInputs/Helpers/SimulationPosition()
     B=&RotateVectorByQuaternion(Vector To Rotate=<0 0 0>, Quaternion=$Engine.Owner.Rotation)
    )
    Ribbon Width=&Multiply_Float(A=$User.RingSize, B=300)
   )
   &AddVelocity(Velocity=&RotateVectorByQuaternion(Vector To Rotate=<0 0 8000>, Quaternion=$Engine.Owner.Rotation)) {enabled=false}
   &SystemLocation(MaskSpawnGroup=false, Offset=<-10 0 0>) {enabled=false}
   &InheritVelocity(Module.Limit Speed=true, Velocity Limit=1000, Velocity Scale=<-100 -100 -100>) {enabled=false, module=InheritVelocity001}
   &InheritVelocity(Module.Limit Speed=true, Velocity Limit=1000, Velocity Scale=<-100 -100 -100>) {enabled=false}
  ]
  particleUpdateScript [
   &ParticleState()
   &CurlNoiseForce(
    Noise Frequency=10
    Noise Strength=&FloatFromCurve(
     FloatCurve=DataInterface() {propertyValues {Curve {keys [{tangentMode=RCTM_User, time=0, value=1}, {tangentMode=RCTM_User, time=1, value=0}]}}}
     Scale Curve=500
    )
    Pan Noise=true
    Pan Noise Field=&MakeVector(
     X=500
     Y=250
     Z=&FloatFromCurve(
      FloatCurve=DataInterface() {propertyValues {Curve {keys [{tangentMode=RCTM_User, time=0, value=0}, {tangentMode=RCTM_User, time=1, value=1}]}}}
      CurveIndex=$Emitter.NormalizedLoopAge
      Scale Curve=5
     )
    )
   ) {enabled=false}
   &InheritSourceMovement(Applied Force Scale=<-1.75 -1.75 -1.75>, ApplyForce=true, ApplyForceFalloff=false, ApplyPosition=false, Module.LimitForceSpeed=false)
   &SolveForcesAndVelocity(Clamp Velocity=false, Limit Acceleration=false)
   &ScaleColor(
    Scale Alpha=&FloatFromCurve(
     FloatCurve=DataInterface() {propertyValues {Curve {keys [{interpMode=RCIM_Cubic, tangentMode=RCTM_User, time=0, value=1}, {interpMode=RCIM_Cubic, tangentMode=RCTM_User, time=1, value=0}]}}}
     Scale Curve=0.1
    )
    Scale RGB=&VectorFromFloat(Value=1)
    ScaleA=true
    ScaleRGB=true
   )
   &ScaleColor(
    Color Value To Scale=$Particles.Color
    Scale Alpha=&FloatFromCurve(
     FloatCurve=DataInterface() {propertyValues {Curve {keys [{interpMode=RCIM_Cubic, tangentMode=RCTM_User, time=0, value=0}, {interpMode=RCIM_Cubic, tangentMode=RCTM_User, time=0.2, value=1}]}}}
     CurveIndex=$Emitter.NormalizedLoopAge
    )
    Scale RGB=&VectorFromFloat(Value=8)
    ScaleA=true
    ScaleRGB=true
   ) {enabled=false, module=ScaleColor001}
   &ScaleRibbonWidth(Ribbon Width Scale=&FloatFromCurve(FloatCurve=DataInterface() {propertyValues {Curve {keys [{interpMode=RCIM_Cubic, tangentMode=RCTM_User, time=0, value=1}, {interpMode=RCIM_Cubic, tangentMode=RCTM_User, time=1, value=0.5}]}}}))
   &DynamicMaterialParameters(
    Index 0 Param 1=&RandomRangeFloat(Minimum=0.8, Maximum=0.8)
    Index 0 Param 2=1
    Index 0 Param 3=2
    Module.Param0WriteEnabled=true
    Module.Param1WriteEnabled=true
    Module.Param2WriteEnabled=true
    Module.Param3WriteEnabled=true
   )
  ]
  rendererClasses [&NiagaraRibbonRendererProperties]
  renderers [{rendererClass=&NiagaraRibbonRendererProperties}]
 }}
summary {
 systemName=NS_KnockbackWeapon_Projectile_Trail
 emitters [{emitterName=Trail, rendererClasses [&NiagaraRibbonRendererProperties]}]
 userVariables [
  User.ADS {type=NiagaraFloat, defaultValue=0, description=""}
  User.Color_01 {type=LinearColor, description="", defaultValue {struct=LinearColor, value=<rgba 0 1 0.8115 1>}}
  User.Color_02 {type=LinearColor, description="", defaultValue {struct=LinearColor, value=<rgba 1 0.0396 0.894 1>}}
  User.Color_03 {type=LinearColor, description="", defaultValue {struct=LinearColor, value=<rgba 0.0664 0.2855 0.625 1>}}
  User.ProjectileScale {type=NiagaraFloat, defaultValue=0, description=""}
  User.RingSize {type=NiagaraFloat, defaultValue=1.35, description=""}
 ]
}
